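; Gameplay ability for the Harpoon Elimination feat that tags actors hit by the harpoon with a gameplay effect.
; On activation it calls the parent, then listens repeatedly for non-periodic gameplay effects the owner applies, filtered by harpoon source tag and target tag requirements.
; Each match calls HarpoonHit, which applies HarpoonHitGE to that target.
; The elimination check itself is not in this graph.
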(event Ability|EventActivateAbility
  (|CallParentFunction)
  (bind (_asynctaskproxy _target) (Ability|Tasks|WaitGameplayEffectAppliedToTarget :SourceTagRequirements (Variables|Default|GetHarpoonWeaponSourceTag) :TargetTagRequirements (Variables|Default|GetTargetTags) :TriggerOnce false :ListenForPeriodicEffects false)
    (:OnApplied
      (CallFunction|HarpoonHit _target))))

(event Ability|EventFailedToActivatePassiveAbility
  (params (ActorInfo :type Gameplay%20Ability%20Actor%20Info%20Structure)))

(event Custom|HarpoonHit
  (params (TargetActor :type Actor%20Object%20Reference%20%28by%20ref%29))
  (Ability|ApplyGameplayEffectToTarget (Ability|TargetData|AbilityTargetDatafromActor TargetActor) (Variables|Default|GetHarpoonHitGE)))
